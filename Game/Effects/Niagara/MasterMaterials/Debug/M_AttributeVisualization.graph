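# Unlit masked debug material for Niagara sprite and mesh particles that visualizes particle attributes as grayscale or printed numbers.
# DebugScalarValues draws the vector length of dynamic parameter float01 replicated to three components, so the number is its magnitude scaled by √3.
# Numeric_Display blends from particle red as grayscale to digits tinted by particle color, with a matching solid-to-digit mask blend.
# A DebugFloat4Values branch on a second dynamic parameter feeds nothing.
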Material /Game/Effects/Niagara/MasterMaterials/Debug/M_AttributeVisualization
graph {
 properties {
  bUsedWithNiagaraMeshParticles=true
  bUsedWithNiagaraSprites=true
  blendMode=BLEND_Masked
  pixelDepthOffsetMode=PDOM_Legacy
  shadingModel=MSM_Unlit
  usedShadingModels=Unlit
 }
 nodes [
  n1 = ParticleColor()
  n2 = DynamicParameter() {paramNames [float01,Param2,Param3,Param4]}
  n3 = MaterialFunctionCall(
   Number=MaterialFunctionCall(Vector 3=MaterialFunctionCall(X=@n2.float01, Y=@n2.float01, Z=@n2.float01) {materialFunction=&MakeFloat3}.Result) {materialFunction=/Engine/Functions/Engine_MaterialFunctions02/Utility/VectorLength}.V3 Length
   MaximumNumberOfDigits=ScalarParameter() {defaultValue=3, parameterName=Digits}
   UVs=TextureCoordinate()
  ) {materialFunction=&DebugScalarValues}
  n4 = ScalarParameter() {defaultValue=1, parameterName=Numeric_Display, sliderMax=1}
  LinearInterpolate(
   A=MaterialFunctionCall(X=@n1.R, Y=@n1.R, Z=@n1.R) {materialFunction=&MakeFloat3}.Result
   B=Multiply(A=@n3.Result, B=@n1.RGB)
   Alpha=@n4
  )
  LinearInterpolate(B=@n3.Result, Alpha=@n4) {constA=1}
  n5 = DynamicParameter() {parameterIndex=1}
  MaterialFunctionCall!(
   Vector4=MaterialFunctionCall(X=@n5.R, Y=@n5.R, Z=@n5.R, A=@n5.R) {materialFunction=&MakeFloat4}.Result
   MaximumNumberOfDigits
   UVs
   DebugTextLocation RG_UpperRight BA_LowerLeft
   Component Spacing
  ) {materialFunction=/Engine/Functions/Engine_MaterialFunctions02/Utility/DebugFloat4Values}
 ]
}
