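# Unlit, two-sided translucent debug master material for inspecting UVs on Niagara mesh particles and ribbons.
# A UV checker texture is sampled on UV channels 0 and 1, blended by a channel parameter.
# A power-curved weight tints the checker with particle color, fading it to pure particle color at full weight.
# An optional SplineThicken world position offset animates width between two parameters via DebugTimeSine, with a UV swap switch.
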Material /Game/Effects/Niagara/MasterMaterials/Debug/M_DebugUV
graph {
 properties {
  bUsedWithNiagaraMeshParticles=true
  bUsedWithNiagaraRibbons=true
  blendMode=BLEND_Translucent
  pixelDepthOffsetMode=PDOM_Legacy
  shadingModel=MSM_Unlit
  twoSided=true
  usedShadingModels=Unlit
 }
 nodes [
  n1 = Constant3Vector() {constant=<rgba 1 1 1 0>}
  n2 = Power(Base=ScalarParameter() {parameterName=Particle Color Mix Weight, sliderMax=1}) {constExponent=2.2}
  n3 = ParticleColor()
  n4 = TextureSample(UVs=TextureCoordinate()) {texture=/Game/Packages/Fortress_SharedMaterials/Textures/T_UV_Checker_D}
  n5 = TextureSample(UVs=TextureCoordinate() {coordinateIndex=1}) {texture=&T_UV_Checker_D}
  n6 = ScalarParameter() {defaultValue=1, parameterName=UV Channel, sliderMax=1}
  LinearInterpolate(A=Multiply(A=LinearInterpolate(A=@n1, B=@n3.RGB, Alpha=@n2), B=LinearInterpolate(A=@n4.RGB, B=@n5.RGB, Alpha=@n6)), B=@n3.RGB, Alpha=@n2)
  n7 = LinearInterpolate(
   A=ScalarParameter() {parameterName=Width Low}
   B=ScalarParameter() {defaultValue=10, parameterName=Width}
   Alpha=MaterialFunctionCall() {materialFunction=&DebugTimeSine}.Result
  ) {constA=10}
  LinearInterpolate(
   B=MaterialFunctionCall(WidthBase=@n7, WidthTip=@n7, Expand U or V UV Channel=StaticBoolParameter() {parameterName=Swap UV}) {materialFunction=&SplineThicken}.WorldPositionOffset
   Alpha=ScalarParameter() {parameterName=Debug SplineThicken}
  )
 ]
}
