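# M_DebugWater is an unlit translucent debug material for Niagara sprite and mesh particles that visualizes water waves.
# Dynamic parameters supply water body index, depth and SubUV frame data, with a DebugSubUV flipbook blended over particle color and an opacity floor.
# A static switch enables displacement from Z-only Gerstner waves scaled by depth-based attenuation.
# The attenuation function's Water Body Index input receives Depth, not the index.
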Material /Game/Effects/Niagara/MasterMaterials/Debug/M_DebugWater
graph {
 properties {
  bAutomaticallySetUsageInEditor=false
  bUsedWithNiagaraMeshParticles=true
  bUsedWithNiagaraSprites=true
  blendMode=BLEND_Translucent
  shadingModel=MSM_Unlit
  usedShadingModels=Unlit
 }
 nodes [
  n1 = DynamicParameter() {paramNames [WaterBody Index, Depth, SubUV_Images, SubUV_Frame]}
  n2 = MaterialFunctionCall!(
   Texture=TextureObject() {texture=&/DebugSubUV}
   UVs=TextureCoordinate()
   SubImages=AppendVector(A=@n1.SubUV_Images, B=@n1.SubUV_Images)
   Frame=@n1.SubUV_Frame
  ) {materialFunction=/Engine/Functions/Engine_MaterialFunctions03/Texturing/SubUV_Function}
  Add(
   A=Multiply(A=ParticleColor().RGB, B=OneMinus(None=@n2.Alpha))
   B=Multiply(A=@n2.RGB, B=ScalarParameter() {defaultValue=0.2, parameterName=ColorScale})
  )
  Max(A=@n2.Alpha, B=ScalarParameter() {defaultValue=0.5, parameterName=DebugOpacityBase})
  n3 = StaticBool() {value=true}
  n4 = WorldPosition() {worldPositionShaderOffset=WPT_ExcludeAllShaderOffsets}
  n5 = DynamicParameter() {paramNames [WaterBody Index, Depth, SubUV_Images, SubUV_Frame]}
  StaticSwitchParameter(
   True=Multiply(
    A=MaterialFunctionCall!(Z Offset Only=@n3, Position=@n4.XYZ, WaterBodyIndex=@n5.WaterBody Index, Normals use VS Interpolation, All Normals VS) {materialFunction=/FortWater/Materials/Functions/Fort_ComputeGerstnerWaves}.WPO
    B=MaterialFunctionCall!(WaterDepth=@n5.Depth, Water Body Index=@n5.Depth) {materialFunction=/Water/Materials/Functions/ComputeWavesAttenuationFactor}.WaveAttenuationFactor
   )
   False=Constant()
  ) {defaultValue=true, parameterName=UseWaterDisplacement}
 ]
}
