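# Unlit debug material for Niagara sprite and mesh particles that draws values from Particle Color as on-screen digits.
# A static switch "RGBA or Float" selects DebugScalarValues showing the red channel when true, or DebugFloat4Values showing color-coded RGBA when false.
# A Digits parameter caps the digit count, and a Scale parameter scales the UVs around their center to size the text.
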Material /Game/Effects/Niagara/MasterMaterials/Debug/M_ParticleDebugValue
graph {
 properties {
  bUsedWithNiagaraMeshParticles=true
  bUsedWithNiagaraSprites=true
  pixelDepthOffsetMode=PDOM_Legacy
  shadingModel=MSM_Unlit
  usedShadingModels=Unlit
 }
 nodes [
  n1 = ParticleColor()
  n2 = ScalarParameter() {defaultValue=4, parameterName=Digits, sliderMax=5}
  n3 = TextureCoordinate()
  n4 = Constant() {r=0.5}
  n5 = Add(A=Multiply(A=Subtract(A=@n3, B=@n4), B=ScalarParameter() {defaultValue=0.5, parameterName=Scale}), B=@n4)
  StaticSwitch(
   True=MaterialFunctionCall(Number=@n1.R, MaximumNumberOfDigits=@n2, UVs=@n5) {materialFunction=&DebugScalarValues}.Result
   False=MaterialFunctionCall!(Vector4=@n1.RGBA, MaximumNumberOfDigits=@n2, UVs=@n5, DebugTextLocation RG_UpperRight BA_LowerLeft, Component Spacing) {materialFunction=/Engine/Functions/Engine_MaterialFunctions02/Utility/DebugFloat4Values}.ColorCodedOutput
   Value=StaticBoolParameter() {parameterName=RGBA or Float}
  )
 ]
}
